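# Sound cue for the tier 4 axe heavy attack, using small melee weapon attenuation, a melee sound class and player melee concurrency.
# Each branch mixes a random axe swing, pitched down, with the GreatSword Revenant super attack wave, pitched up.
# A quality-level root selects between two such branches, the first with four swing variants and the second with three, leaving its other two slots empty.
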SoundCue /Game/Sounds/Fort_Weapons/Axes/Axe_T04_HeavyAttack_Cue
root=QualityLevel_0
nodes [
 {id=Mixer_1, children [Modulator_2,Modulator_3], properties {inputVolume [1,1]}}
 {id=Mixer_2, children [Modulator_4,Modulator_5], properties {inputVolume [1,1]}}
 {id=Modulator_2, children [Random_1], properties {pitchMax=0.9, pitchMin=0.8}}
 {id=Modulator_3, children [WavePlayer_9], properties {pitchMax=1.4, pitchMin=1.2}}
 {id=Modulator_4, children [Random_2], properties {pitchMax=0.9, pitchMin=0.8}}
 {id=Modulator_5, children [WavePlayer_14], properties {pitchMax=1.4, pitchMin=1.2}}
 {id=QualityLevel_0, children [Mixer_1,Mixer_2,null,null]}
 {id=Random_1, children [WavePlayer_5,WavePlayer_6,WavePlayer_7,WavePlayer_8], properties {weights [1,1,1,1]}}
 {id=Random_2, children [WavePlayer_11, WavePlayer_12, WavePlayer_13], properties {weights [1,1,1]}}
 {id=WavePlayer_11, properties {soundWaveAssetPtr=&/Axe_New_Swing_03}}
 {id=WavePlayer_12, properties {soundWaveAssetPtr=&/Axe_New_Swing_02}}
 {id=WavePlayer_13, properties {soundWaveAssetPtr=&/Axe_New_Swing_01}}
 {id=WavePlayer_14, properties {soundWaveAssetPtr=/Game/Sounds/Fort_Weapons/GreatSword/Revenant_SuperAttack}}
 {id=WavePlayer_5, properties {soundWaveAssetPtr=&/Axe_New_Swing_04}}
 {id=WavePlayer_6, properties {soundWaveAssetPtr=&Axe_New_Swing_03}}
 {id=WavePlayer_7, properties {soundWaveAssetPtr=&Axe_New_Swing_02}}
 {id=WavePlayer_8, properties {soundWaveAssetPtr=&Axe_New_Swing_01}}
 {id=WavePlayer_9, properties {soundWaveAssetPtr=&Revenant_SuperAttack}}
]
properties {
 attenuationSettings=&Fort_Weapon_Melee_Sml_Att
 maxDistance=2800
 soundClassObject=&Fort_WeaponMelee_PSM
 volumeMultiplier=0.8
 concurrencySet [&Player_Melee_Weapon_Conc]
}
