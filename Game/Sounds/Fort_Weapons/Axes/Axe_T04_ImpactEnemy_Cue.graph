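# Sound Cue for a tier 4 axe hitting an enemy, using small melee weapon attenuation and melee concurrency.
# A quality switch fills two of four slots, each mixing a random axe impact with a quieter random Revenant greatsword impact, both pitch-varied.
# The second slot draws on fewer waves, and its greatsword random node has an empty entry.
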SoundCue /Game/Sounds/Fort_Weapons/Axes/Axe_T04_ImpactEnemy_Cue
root=QualityLevel_0
nodes [
 {id=Mixer_1, children [Modulator_2,Modulator_3], properties {inputVolume [1,0.7]}}
 {id=Mixer_2, children [Modulator_4,Modulator_5], properties {inputVolume [1,0.7]}}
 {id=Modulator_2, children [Random_2], properties {pitchMax=1.1, pitchMin=0.9, volumeMax=0.7, volumeMin=0.7}}
 {id=Modulator_3, children [Random_3], properties {pitchMax=1.1, pitchMin=0.9, volumeMax=0.7, volumeMin=0.7}}
 {id=Modulator_4, children [Random_4], properties {pitchMax=1.1, pitchMin=0.9, volumeMax=0.7, volumeMin=0.7}}
 {id=Modulator_5, children [Random_5], properties {pitchMax=1.1, pitchMin=0.9, volumeMax=0.7, volumeMin=0.7}}
 {id=QualityLevel_0, children [Mixer_1,Mixer_2,null,null]}
 {id=Random_2, children [WavePlayer_10, WavePlayer_11, WavePlayer_12, WavePlayer_13], properties {weights [1,1,1,1]}}
 {
  id=Random_3
  children [WavePlayer_14, WavePlayer_15, WavePlayer_16, WavePlayer_17, WavePlayer_18, WavePlayer_19]
  properties {weights [1,1,1,1,1,1]}
 }
 {id=Random_4, children [WavePlayer_21, WavePlayer_22, WavePlayer_23], properties {weights [1,1,1]}}
 {id=Random_5, children [WavePlayer_24, WavePlayer_25, WavePlayer_26, WavePlayer_27, null], properties {weights [1,1,1,1,1]}}
 {id=WavePlayer_10, properties {soundWaveAssetPtr=&/Axe_New_Impact_04}}
 {id=WavePlayer_11, properties {soundWaveAssetPtr=&/Axe_New_Impact_03}}
 {id=WavePlayer_12, properties {soundWaveAssetPtr=&/Axe_New_Impact_02}}
 {id=WavePlayer_13, properties {soundWaveAssetPtr=&/Axe_New_Impact_01}}
 {id=WavePlayer_14, properties {soundWaveAssetPtr=/Game/Sounds/Fort_Weapons/GreatSword/Revenant_EnemyImpact_001}}
 {id=WavePlayer_15, properties {soundWaveAssetPtr=/Game/Sounds/Fort_Weapons/GreatSword/Revenant_EnemyImpact_002}}
 {id=WavePlayer_16, properties {soundWaveAssetPtr=/Game/Sounds/Fort_Weapons/GreatSword/Revenant_EnemyImpact_003}}
 {id=WavePlayer_17, properties {soundWaveAssetPtr=/Game/Sounds/Fort_Weapons/GreatSword/Revenant_EnemyImpact_004}}
 {id=WavePlayer_18, properties {soundWaveAssetPtr=/Game/Sounds/Fort_Weapons/GreatSword/Revenant_EnemyImpact_005}}
 {id=WavePlayer_19, properties {soundWaveAssetPtr=/Game/Sounds/Fort_Weapons/GreatSword/Revenant_EnemyImpact_006}}
 {id=WavePlayer_21, properties {soundWaveAssetPtr=&Axe_New_Impact_03}}
 {id=WavePlayer_22, properties {soundWaveAssetPtr=&Axe_New_Impact_02}}
 {id=WavePlayer_23, properties {soundWaveAssetPtr=&Axe_New_Impact_01}}
 {id=WavePlayer_24, properties {soundWaveAssetPtr=&Revenant_EnemyImpact_001}}
 {id=WavePlayer_25, properties {soundWaveAssetPtr=&Revenant_EnemyImpact_002}}
 {id=WavePlayer_26, properties {soundWaveAssetPtr=&Revenant_EnemyImpact_003}}
 {id=WavePlayer_27, properties {soundWaveAssetPtr=&Revenant_EnemyImpact_004}}
]
properties {
 attenuationSettings=&Fort_Weapon_Melee_Sml_Att
 maxDistance=2800
 soundClassObject=&Fort_WeaponMelee_PSM
 volumeMultiplier=1.2
 concurrencyOverrides {resolutionRule=PreventNew}
 concurrencySet [&Player_Melee_Weapon_Conc]
}
