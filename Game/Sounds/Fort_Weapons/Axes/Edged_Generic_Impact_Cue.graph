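# Sound cue for a generic edged melee weapon impact, built from reused Fire_Axe_Enemy waves.
# A quality-level node selects between two equal-weight random pools, four waves in the first and three in the second, leaving its other two inputs empty.
# Both pools pass through identical modulators that apply a fixed reduction in pitch and volume.
# Small-melee attenuation, a melee sound class and a player melee concurrency set govern playback.
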SoundCue /Game/Sounds/Fort_Weapons/Axes/Edged_Generic_Impact_Cue
root=QualityLevel_0
nodes [
 {id=Modulator_1, children [Random_1], properties {pitchMax=0.7, pitchMin=0.7, volumeMax=0.3, volumeMin=0.3}}
 {id=Modulator_2, children [Random_2], properties {pitchMax=0.7, pitchMin=0.7, volumeMax=0.3, volumeMin=0.3}}
 {id=QualityLevel_0, children [Modulator_1,Modulator_2,null,null]}
 {id=Random_1, children [WavePlayer_4,WavePlayer_5,WavePlayer_6,WavePlayer_7], properties {weights [1,1,1,1]}}
 {id=Random_2, children [WavePlayer_8, WavePlayer_9, WavePlayer_10], properties {weights [1,1,1]}}
 {id=WavePlayer_10, properties {soundWaveAssetPtr=&/Fire_Axe_Enemy_004}}
 {id=WavePlayer_4, properties {soundWaveAssetPtr=&/Fire_Axe_Enemy_001}}
 {id=WavePlayer_5, properties {soundWaveAssetPtr=&/Fire_Axe_Enemy_003}}
 {id=WavePlayer_6, properties {soundWaveAssetPtr=&Fire_Axe_Enemy_004}}
 {id=WavePlayer_7, properties {soundWaveAssetPtr=&/Fire_Axe_Enemy_005}}
 {id=WavePlayer_8, properties {soundWaveAssetPtr=&Fire_Axe_Enemy_001}}
 {id=WavePlayer_9, properties {soundWaveAssetPtr=&Fire_Axe_Enemy_003}}
]
properties {
 attenuationSettings=&Fort_Weapon_Melee_Sml_Att
 maxDistance=2800
 soundClassObject=&Fort_WeaponMelee_PSM
 volumeMultiplier=1
 concurrencyOverrides {resolutionRule=PreventNew}
 concurrencySet [&Player_Melee_Weapon_Conc]
}
