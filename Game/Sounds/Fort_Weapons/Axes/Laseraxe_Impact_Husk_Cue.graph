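# Sound cue for a laser axe hitting a Husk, played as a spatialized melee weapon sound with shared concurrency.
# A quality level node picks one of three mixer branches, with variation pools shrinking down the levels and the smallest branch reused for the last two slots.
# Each branch layers a modulated random laser axe impact at reduced volume with a Husk impact delayed by a fixed fraction of a second.
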SoundCue /Game/Sounds/Fort_Weapons/Axes/Laseraxe_Impact_Husk_Cue
root=QualityLevel_0
nodes [
 {id=Delay_0, children [WavePlayer_3], properties {delayMax=0.2, delayMin=0.2}}
 {id=Delay_1, children [Random_3], properties {delayMax=0.2, delayMin=0.2}}
 {id=Delay_2, children [Random_5], properties {delayMax=0.2, delayMin=0.2}}
 {id=Mixer_0, children [Modulator_0,Delay_0], properties {inputVolume [0.7,1]}}
 {id=Mixer_1, children [Modulator_1,Delay_1], properties {inputVolume [0.7,1]}}
 {id=Mixer_2, children [Modulator_2,Delay_2], properties {inputVolume [0.7,1]}}
 {id=Modulator_0, children [Random_0]}
 {id=Modulator_1, children [Random_2]}
 {id=Modulator_2, children [Random_4]}
 {id=QualityLevel_0, children [Mixer_1,Mixer_2,Mixer_0,Mixer_0]}
 {id=Random_0, children [WavePlayer_0,WavePlayer_1], properties {weights [1,1]}}
 {
  id=Random_2
  children [WavePlayer_12, WavePlayer_13, WavePlayer_14, WavePlayer_15, WavePlayer_16, WavePlayer_17]
  properties {weights [1,1,1,1,1,1]}
 }
 {
  id=Random_3
  children [WavePlayer_18, WavePlayer_19, WavePlayer_20, WavePlayer_21, WavePlayer_22, WavePlayer_23]
  properties {weights [1,1,1,1,1,1]}
 }
 {id=Random_4, children [WavePlayer_26, WavePlayer_28, WavePlayer_29], properties {weights [1,1,1]}}
 {id=Random_5, children [WavePlayer_30, WavePlayer_31, WavePlayer_35], properties {weights [1,1,1]}}
 {id=WavePlayer_0, properties {soundWaveAssetPtr=&/laseraxe_impact_03}}
 {id=WavePlayer_1, properties {soundWaveAssetPtr=&/laseraxe_impact_05}}
 {id=WavePlayer_12, properties {soundWaveAssetPtr=&/laseraxe_impact_01}}
 {id=WavePlayer_13, properties {soundWaveAssetPtr=&/laseraxe_impact_02}}
 {id=WavePlayer_14, properties {soundWaveAssetPtr=&laseraxe_impact_03}}
 {id=WavePlayer_15, properties {soundWaveAssetPtr=&/laseraxe_impact_04}}
 {id=WavePlayer_16, properties {soundWaveAssetPtr=&laseraxe_impact_05}}
 {id=WavePlayer_17, properties {soundWaveAssetPtr=&/laseraxe_impact_06}}
 {id=WavePlayer_18, properties {soundWaveAssetPtr=/Game/Sounds/Fort_Monster_Husk/SFX/husk_impact_mix_ST_09}}
 {id=WavePlayer_19, properties {soundWaveAssetPtr=/Game/Sounds/Fort_Monster_Husk/SFX/husk_impact_mix_ST_13}}
 {id=WavePlayer_20, properties {soundWaveAssetPtr=/Game/Sounds/Fort_Monster_Husk/SFX/husk_impact_mix_ST_01}}
 {id=WavePlayer_21, properties {soundWaveAssetPtr=/Game/Sounds/Fort_Monster_Husk/SFX/husk_impact_mix_ST_07}}
 {id=WavePlayer_22, properties {soundWaveAssetPtr=/Game/Sounds/Fort_Monster_Husk/SFX/husk_impact_mix_ST_10}}
 {id=WavePlayer_23, properties {soundWaveAssetPtr=/Game/Sounds/Fort_Monster_Husk/SFX/husk_impact_mix_ST_04}}
 {id=WavePlayer_26, properties {soundWaveAssetPtr=&laseraxe_impact_03}}
 {id=WavePlayer_28, properties {soundWaveAssetPtr=&laseraxe_impact_05}}
 {id=WavePlayer_29, properties {soundWaveAssetPtr=&laseraxe_impact_06}}
 {id=WavePlayer_3, properties {soundWaveAssetPtr=&husk_impact_mix_ST_09}}
 {id=WavePlayer_30, properties {soundWaveAssetPtr=&husk_impact_mix_ST_09}}
 {id=WavePlayer_31, properties {soundWaveAssetPtr=&husk_impact_mix_ST_13}}
 {id=WavePlayer_35, properties {soundWaveAssetPtr=&husk_impact_mix_ST_04}}
]
properties {
 attenuationSettings=&Fort_Weapon_Melee_Sml_Att
 maxDistance=2800
 soundClassObject=&Fort_WeaponMelee_PSM
 volumeMultiplier=0.5
 concurrencySet [&Player_Melee_Weapon_Conc]
}
